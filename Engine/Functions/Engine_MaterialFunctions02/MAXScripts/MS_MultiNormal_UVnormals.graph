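# Material function for the 3ds Max MultiNormal script that outputs vertex normals stored in UVs.
# Custom UV 4 and 5 outputs decode UV1 by unpacking XY, deriving Z and signing it by UV1.R; the rebuilt vector feeds red into both X and Y.
# The Tangent Space UVs output reads a vector back from UV4 and UV5.R, transforms it from local to tangent space and normalizes it.
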MaterialFunction /Engine/Functions/Engine_MaterialFunctions02/MAXScripts/MS_MultiNormal_UVnormals
graph {
 properties {
  bExposeToLibrary=true
  description="This material function was designed to be used in conjunction with the MAXScript located here \"\\UE4\\Engine\\Extras\\3dsMaxScripts\\MultiNormal.ms\". This particular function outputs vertex normals stored in the UVs. "
  libraryCategoriesText [Misc,MAXScripts]
 }
 nodes [
  n1 = AppendVector(
   A=TextureCoordinate() {coordinateIndex=4}
   B=MaterialFunctionCall(Float2=TextureCoordinate() {coordinateIndex=5}) {materialFunction=&BreakOutFloat2Components}.R
  )
  FunctionOutput(None=Normalize(VectorInput=Transform(None=@n1) {transformSourceType=TRANSFORMSOURCE_Local, transformType=TRANSFORM_Tangent})) {outputName=Tangent Space UVs}
  n2 = MaterialFunctionCall(Float2=TextureCoordinate() {coordinateIndex=1}) {materialFunction=&BreakOutFloat2Components}
  n3 = AppendVector(A=Abs(None=@n2.R), B=@n2.G)
  n4 = MaterialFunctionCall(Float3=DeriveNormalZ(InXY=ConstantBiasScale(None=@n3) {bias=-0.5, scale=2})) {materialFunction=&BreakOutFloat3Components}
  n5 = Constant()
  n6 = Constant() {r=1}
  n7 = Constant() {r=-1}
  n8 = MaterialFunctionCall(X=@n4.R, Y=@n4.R, Z=Multiply(A=If(A=@n2.R, B=@n5, "A > B"=@n6, "A == B"=@n7, "A < B"=@n7), B=@n4.B)) {materialFunction=&MakeFloat3}
  FunctionOutput(None=@n8.Result) {outputName=Custom UV 4, sortPriority=1}
  FunctionOutput(None=MaterialFunctionCall(Float3=@n8.Result) {materialFunction=&BreakOutFloat3Components}.B) {outputName=Custom UV 5, sortPriority=2}
 ]
}
