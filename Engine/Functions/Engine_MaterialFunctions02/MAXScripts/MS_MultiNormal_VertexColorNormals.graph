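# Material function that decodes vertex normals stored in vertex colors, for use with the 3ds Max MultiNormal MAXScript.
# It remaps the vertex color from the 0–1 range to -1–1 with a bias and scale, then transforms it from local to tangent space.
# The normalized result is output as vertex color based tangent normals, and the function is exposed to the library under Misc and MAXScripts.
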MaterialFunction /Engine/Functions/Engine_MaterialFunctions02/MAXScripts/MS_MultiNormal_VertexColorNormals
graph {
 properties {
  bExposeToLibrary=true
  description="This material function was designed to be used in conjunction with the MAXScript located here \"\\UE4\\Engine\\Extras\\3dsMaxScripts\\MultiNormal.ms\". This particular function outputs vertex normals stored in the Vertex Colors. "
  libraryCategoriesText [Misc,MAXScripts]
 }
 nodes [
  n1 = ConstantBiasScale(None=VertexColor()) {bias=-0.5, scale=2}
  FunctionOutput(None=Normalize(VectorInput=Transform(None=@n1) {transformSourceType=TRANSFORMSOURCE_Local, transformType=TRANSFORM_Tangent})) {outputName=Vertex Color Based Tangent Normals}
 ]
}
