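# Material function in the MAXScripts library category that reads a Sequence Painter sequence value from mesh UVs.
# It breaks texture coordinate index 2 into components, takes the R (U) component, and rounds it up with Ceil as the single output.
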MaterialFunction /Engine/Functions/Engine_MaterialFunctions02/MAXScripts/MS_SequencePainter_Sequence
graph {
 properties {bExposeToLibrary=true, libraryCategoriesText [Misc,MAXScripts]}
 nodes [
  n1 = TextureCoordinate() {coordinateIndex=2}
  FunctionOutput(None=Ceil(None=MaterialFunctionCall(Float2=@n1) {materialFunction=&BreakOutFloat2Components}.R))
 ]
}
