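# Material function that plays a mesh flipbook from static meshes processed by the Sequence Painter section of the Texture Morph MAXScript.
# The rounded-up U of UV channel 2 gives each vertex's frame number, compared to the rounded-up product of the 0-1 animation input and frame count.
# A SphereMask matches the two, and the World Position Offset output collapses non-matching vertices onto the component pivot, hiding inactive frames.
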MaterialFunction /Engine/Functions/Engine_MaterialFunctions02/MAXScripts/MS_SequencePainter_SequenceFlipbook
graph {
 properties {
  bExposeToLibrary=true
  description="This function will create a mesh flipbook from assets processed with the \"Sequence Painter\" section of the \"Texture Morph\"  MAXScript."
  libraryCategoriesText [Misc,MAXScripts]
 }
 nodes [
  n1 = TextureCoordinate() {coordinateIndex=2}
  n2 = FunctionInput(Preview=MaterialFunctionCall() {materialFunction=&TimeWithSpeedVariable}.Frac Time) {bUsePreviewValueAsDefault=true, inputName=0-1 Animation, inputType=FunctionInput_Scalar, previewValue=<5 0 0 1>}
  n3 = FunctionInput() {
   bUsePreviewValueAsDefault=true
   description=Enter the number of static meshes that were processed by the script.
   inputName=Number Of Frames
   inputType=FunctionInput_Scalar
   sortPriority=1
   previewValue=<5 0 0 1>
  }
  n4 = SphereMask(
   A=Ceil(None=MaterialFunctionCall(Float2=@n1) {materialFunction=&BreakOutFloat2Components}.R)
   B=Ceil(None=Multiply(A=@n2, B=@n3) {constB=25})
   Radius=Constant() {r=0.5}
  )
  n5 = MaterialFunctionCall() {materialFunction=&ComponentPivotLocation}
  n6 = WorldPosition()
  FunctionOutput(None=Multiply(A=OneMinus(None=@n4), B=Subtract(A=@n5.Result, B=@n6.XYZ))) {outputName=World Position Offset}
 ]
}
