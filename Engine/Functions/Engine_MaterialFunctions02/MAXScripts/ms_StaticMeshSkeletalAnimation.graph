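# Material function that plays baked skeletal animation on static meshes by decoding textures and bone indices packed by the StaticMeshSkeletalAnimation 3ds Max script.
# Animation phase and bone index form the UV into position and rotation textures.
# World position offset adds sampled translation to rotation about the bone's rest-pose pivot, all transformed from instance space.
# A separate pixel-shader bone lookup corrects normals, and rotation axis and angle are also output.
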MaterialFunction /Engine/Functions/Engine_MaterialFunctions02/MAXScripts/ms_StaticMeshSkeletalAnimation
graph {
 properties {
  bExposeToLibrary=true
  description="This material function decodes the skeletal mesh animation textures and bone index data packed by the 3DS MAXScript located here: ...Engine\\Extras\\3dsMaxScripts\\StaticMeshSkeletalAnimation.ms. For more information please refer to the guide button with the script and online documentation."
  libraryCategoriesText [MAXScripts]
 }
 nodes [
  n1 = FunctionInput(Preview=Frac(None=Time())) {
   bUsePreviewValueAsDefault=true
   description=Enter a 0-1 value as a lookup into the animation texture.
   inputName=Animation Phase
   inputType=FunctionInput_Scalar
  }
  n2 = FunctionInput(Preview=VertexColor().A) {inputName=Vertex Shader Bone Index, inputType=FunctionInput_Scalar, sortPriority=20}
  n3 = MaterialFunctionCall(X=@n1, Y=@n2) {materialFunction=&MakeFloat2}
  n4 = TextureSample(UVs=@n3.Result, Tex=FunctionInput() {inputName=Position Animation, inputType=FunctionInput_Texture2D, sortPriority=6}) {samplerType=SAMPLERTYPE_LinearColor}
  n5 = FunctionInput() {inputName=Rotation Animation, inputType=FunctionInput_Texture2D, sortPriority=7}
  n6 = TextureSample(UVs=@n3.Result, Tex=@n5) {samplerType=SAMPLERTYPE_LinearColor}
  n7 = MaterialFunctionCall(Vector=Transform(None=@n6.RGB) {transformSourceType=TRANSFORMSOURCE_Instance}, Default=Constant3Vector() {constant=<rgba 0 0 1 0>}) {materialFunction=&SafeNormalize}
  n8 = TransformPosition(None=TextureSample(UVs=@n2, Tex=FunctionInput() {inputName=Rest Pose Texture, inputType=FunctionInput_Texture2D, sortPriority=5}) {samplerType=SAMPLERTYPE_LinearColor}.RGB) {transformSourceType=TRANSFORMPOSSOURCE_Instance, transformType=TRANSFORMPOSSOURCE_World}
  n9 = FunctionInput(Preview=WorldPosition().XYZ) {bUsePreviewValueAsDefault=true, inputName=World Position, sortPriority=99}
  FunctionOutput(None=Add(
    A=Transform(None=@n4.RGB) {transformSourceType=TRANSFORMSOURCE_Instance}
    B=RotateAboutAxis(NormalizedRotationAxis=@n7.Result, RotationAngle=@n6.A, PivotPoint=@n8, Position=@n9)
   )) {outputName=World Position Offset}
  n10 = FunctionInput(Preview=@n2) {
   bUsePreviewValueAsDefault=true
   description=Input the vertex alpha or a channel from the UV that was used to store the bone index in the script.
   inputName=Pixel Shader Bone Index
   inputType=FunctionInput_Scalar
   sortPriority=21
  }
  n11 = TextureSample(UVs=MaterialFunctionCall(X=@n1, Y=@n10) {materialFunction=&MakeFloat2}.Result, Tex=@n5) {samplerType=SAMPLERTYPE_LinearColor}
  n12 = Transform(None=@n11.RGB) {transformSourceType=TRANSFORMSOURCE_Instance}
  n13 = Constant3Vector() {constant=<rgba 0 0 1 0>}
  n14 = FunctionInput(Preview=VertexNormalWS()) {bUsePreviewValueAsDefault=true, inputName=World Space Normals, sortPriority=10, previewValue=<0 0 1 1>}
  FunctionOutput(None=MaterialFunctionCall!(
    Rotation Axis=MaterialFunctionCall(Vector=@n12, Default=@n13) {materialFunction=&SafeNormalize}.Result
    Rotation Angle=@n11.A
    World Space Vertex Normals=@n14
    Movement Mask
    PivotPoint
   ) {materialFunction=&FixRotateAboutAxisNormals}.New World Space Vertex Normal) {outputName=World Space Normals}
  FunctionOutput(None=AppendVector(A=@n7.Result, B=@n6.A)) {outputName=Advanced: Mesh Rotation Axis And Angle, sortPriority=99}
 ]
}
