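# Material function that plays 3ds Max MAXScript-baked skeletal animation on static meshes, with animated normals.
# Animation phase and per-vertex bone index look up position and rotation textures, producing world position offset around rest-pose pivots.
# The same rotation updates the tangent basis, packed into custom UVs 2-6 and rebuilt per pixel to transform the normal map.
# Updated basis vectors are also output so instances can be chained.
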MaterialFunction /Engine/Functions/Engine_MaterialFunctions02/MAXScripts/ms_StaticMeshSkeletalAnimationHighQuality
graph {
 properties {
  bExposeToLibrary=true
  description="This material function decodes the skeletal mesh animation textures and bone index data packed by the 3DS MAXScript located here: ...Engine\\Extras\\3dsMaxScripts\\StaticMeshSkeletalAnimation.ms. When using this material function in a chain, feed the output vectors to the subsequent material function. You'll only need to pipe the normal texture input into the last material function. The last instance of the function should be connected to the base materials custom uv inputs.  For more information please refer to the guide button with the script and online documentation. "
  libraryCategoriesText [MAXScripts]
 }
 nodes [
  n1 = FunctionInput(Preview=Frac(None=Time())) {
   bUsePreviewValueAsDefault=true
   description=Enter a 0-1 value as a lookup into the animation texture.
   inputName=Animation Phase
   inputType=FunctionInput_Scalar
  }
  n2 = FunctionInput(Preview=VertexColor().A) {
   description=Input the vertex alpha or a channel from the UV that was used to store the bone index in the script.
   inputName=Vertex Shader Bone Index
   inputType=FunctionInput_Scalar
   sortPriority=20
  }
  n3 = MaterialFunctionCall(X=@n1, Y=@n2) {materialFunction=&MakeFloat2}
  n4 = TextureSample(UVs=@n3.Result, Tex=FunctionInput() {inputName=Position Animation, inputType=FunctionInput_Texture2D, sortPriority=6}) {samplerType=SAMPLERTYPE_LinearColor}
  n5 = TextureSample(UVs=@n3.Result, Tex=FunctionInput() {inputName=Rotation Animation, inputType=FunctionInput_Texture2D, sortPriority=7}) {samplerType=SAMPLERTYPE_LinearColor}
  n6 = Normalize(VectorInput=Transform(None=@n5.RGB) {transformSourceType=TRANSFORMSOURCE_Instance})
  n7 = TransformPosition(None=TextureSample(UVs=@n2, Tex=FunctionInput() {inputName=Rest Pose Texture, inputType=FunctionInput_Texture2D, sortPriority=5}) {samplerType=SAMPLERTYPE_LinearColor}.RGB) {transformSourceType=TRANSFORMPOSSOURCE_Instance, transformType=TRANSFORMPOSSOURCE_World}
  n8 = FunctionInput(Preview=WorldPosition().XYZ) {bUsePreviewValueAsDefault=true, inputName=World Position, sortPriority=99}
  FunctionOutput(None=Add(
    A=Transform(None=@n4.RGB) {transformSourceType=TRANSFORMSOURCE_Instance}
    B=RotateAboutAxis(NormalizedRotationAxis=@n6, RotationAngle=@n5.A, PivotPoint=@n7, Position=@n8)
   )) {outputName=World Position Offset}
  n9 = FunctionInput() {bUsePreviewValueAsDefault=true, inputName=Tangent Space Normal Map, sortPriority=10, previewValue=<0 0 1 1>}
  n10 = MaterialFunctionCall(Float2=TextureCoordinate() {coordinateIndex=3}) {materialFunction=&BreakOutFloat2Components}
  n11 = AppendVector(A=TextureCoordinate() {coordinateIndex=2}, B=@n10.R)
  n12 = AppendVector(A=@n10.G, B=TextureCoordinate() {coordinateIndex=4})
  n13 = AppendVector(
   A=TextureCoordinate() {coordinateIndex=5}
   B=MaterialFunctionCall(Float2=TextureCoordinate() {coordinateIndex=6}) {materialFunction=&BreakOutFloat2Components}.R
  )
  FunctionOutput(None=MaterialFunctionCall(VectorToTransform=@n9, BasisX=@n11, BasisY=@n12, BasisZ=@n13) {materialFunction=&Transform3x3Matrix}.Transformed Vector) {outputName=World Space Normals}
  n14 = FunctionInput(Preview=Transform(None=Constant3Vector() {constant=<rgba 0 0 1 0>})) {bUsePreviewValueAsDefault=true, inputName=Tangent Basis Z In WS, sortPriority=32}
  n15 = RotateAboutAxis(NormalizedRotationAxis=@n6, RotationAngle=@n5.A, PivotPoint=Constant(), Position=@n14)
  n16 = MaterialFunctionCall(Float3=FunctionOutput(None=Add(A=@n15, B=@n14)) {outputName=Updated Vertex Shader Z Vector, sortPriority=22}) {materialFunction=&BreakOutFloat3Components}
  FunctionOutput(None=MaterialFunctionCall(X=@n16.R, Y=@n16.R) {materialFunction=&MakeFloat2}.Result) {outputName=Custom UV 5, sortPriority=13}
  FunctionOutput(None=@n16.B) {outputName=Custom UV 6, sortPriority=14}
  FunctionOutput(None=@n13) {outputName=Updated Pixel Shader Z Vector, sortPriority=32}
  FunctionOutput(None=@n11) {outputName=Updated Pixel Shader X Vector, sortPriority=30}
  FunctionOutput(None=@n12) {outputName=Updated Pixel Shader Y Vector, sortPriority=31}
  n17 = FunctionInput(Preview=Transform(None=Constant3Vector() {constant=<rgba 1 0 0 0>})) {bUsePreviewValueAsDefault=true, inputName=Tangent Basis X In WS, sortPriority=30}
  n18 = RotateAboutAxis(NormalizedRotationAxis=@n6, RotationAngle=@n5.A, PivotPoint=Constant(), Position=@n17)
  n19 = MaterialFunctionCall(Float3=FunctionOutput(None=Add(A=@n18, B=@n17)) {outputName=Updated Vertex Shader X Vector, sortPriority=20}) {materialFunction=&BreakOutFloat3Components}
  FunctionOutput(None=AppendVector(A=@n19.R, B=@n19.R)) {outputName=Custom UV 2, sortPriority=10}
  n20 = FunctionInput(Preview=Transform(None=Constant3Vector() {constant=<rgba 0 1 0 0>})) {bUsePreviewValueAsDefault=true, inputName=Tangent Basis Y In WS, sortPriority=31}
  n21 = RotateAboutAxis(NormalizedRotationAxis=@n6, RotationAngle=@n5.A, PivotPoint=Constant(), Position=@n20)
  n22 = MaterialFunctionCall(Float3=FunctionOutput(None=Add(A=@n21, B=@n20)) {outputName=Updated Vertex Shader Y Vector, sortPriority=21}) {materialFunction=&BreakOutFloat3Components}
  FunctionOutput(None=MaterialFunctionCall(X=@n19.B, Y=@n22.R) {materialFunction=&MakeFloat2}.Result) {outputName=Custom UV 3, sortPriority=11}
  FunctionOutput(None=MaterialFunctionCall(X=@n22.G, Y=@n22.G) {materialFunction=&MakeFloat2}.Result) {outputName=Custom UV 4, sortPriority=12}
 ]
}
