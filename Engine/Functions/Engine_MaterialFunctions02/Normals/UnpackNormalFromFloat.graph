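# Material function that unpacks normals or float3 vectors from a single scalar, intended for processed normal maps.
# It splits the decimal digits by scaling and rounding, offering XXYY.0, XXX.YYY and XXY.YZZ layouts as separate outputs.
# The two-component layouts derive Z and take its sign from the scalar's sign, while XXY.YZZ yields all three components, raw or normalized.
# Debug value displays and example constants remain in the graph, unconnected to outputs.
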MaterialFunction /Engine/Functions/Engine_MaterialFunctions02/Normals/UnpackNormalFromFloat
graph {
 properties {bExposeToLibrary=true, description=Unpack Normal/Float3 values from a scalar value, libraryCategoriesText [Misc,Math]}
 nodes [
  Constant3Vector() {desc=Deviation amount, constant=<rgba 0.0024 0.0049 0.004 0>}
  Constant3Vector() {desc=Variation, constant=<rgba 0.0064 0.0201 0.4836 0>}
  Constant3Vector() {desc=Result, constant=<rgba 0.34 -0.72 0.6049 0>}
  Constant3Vector() {desc=Target vector, constant=<rgba 0.3261 -0.7259 0.6056 0>}
  Constant3Vector() {desc=Loss Amount, constant=<rgba -0.0139 -0.0059 0.0007 0>}
  Constant3Vector() {desc=Deviation amount, constant=<rgba 0.0024 0.0049 0.004 0>}
  Constant3Vector() {desc=Variation, constant=<rgba 0.0016 0.0013 0.0028 0>}
  Constant3Vector() {desc=Variation, constant=<rgba 0.0017 0.0452 -0.0401 0>}
  Constant3Vector() {desc=Variation, constant=<rgba 0.0016 0.0013 0.0028 0>}
  Constant() {r=221.726}
  Constant() {r=-205.228}
  Constant() {r=-205.228}
  Constant() {desc=input scalar, r=6714}
  n1 = FunctionInput() {
   bUsePreviewValueAsDefault=true
   description=To be used with processed normal maps. Example packed and unpacked values stored within the function.
   inputName=Scalar Value
   inputType=FunctionInput_Scalar
   previewValue=<0.0002 0 0 1>
  }
  n2 = MaterialFunctionCall!(S=@n1, V2, V3, V4) {materialFunction=/Engine/Functions/Engine_MaterialFunctions02/PassThrough}
  n3 = Abs(None=@n2.S)
  n4 = Floor(None=Divide(A=@n3) {constB=100})
  n5 = Divide(A=Floor(None=@n3)) {constB=100}
  n6 = Floor(None=@n5)
  n7 = Subtract(A=AppendVector(A=Divide(A=@n4) {constB=100}, B=Subtract(A=@n5, B=@n6))) {constB=0.5}
  n8 = Add(A=@n7, B=@n7)
  n9 = Multiply(A=DeriveNormalZ(InXY=@n8), B=If(A=@n2.S, B=Constant(), "A > B"=Constant() {r=1}, "A < B"=Constant() {r=-1}))
  n10 = MaterialFunctionCall(Vector=AppendVector(A=@n8, B=ComponentMask(None=@n9) {b=true}), Default=Constant3Vector() {constant=<rgba 0 0 1 0>}) {materialFunction=&SafeNormalize}
  FunctionOutput(None=@n10.Result) {outputName="( z sign ) XXYY.0", sortPriority=-1}
  n11 = Multiply(A=@n1) {constB=0.001}
  n12 = Multiply(A=@n1) {constB=0.1}
  n13 = Subtract(A=@n12, B=Floor(None=@n12))
  n14 = Multiply(A=@n1) {constB=10}
  n15 = Subtract(A=@n14, B=Floor(None=@n14))
  n16 = Subtract(A=Multiply(A=MaterialFunctionCall(X=@n11, Y=@n13, Z=@n15) {materialFunction=&MakeFloat3}.Result) {constB=2})
  FunctionOutput(None=@n16) {outputName=XXY.YZZ Vector, sortPriority=1}
  n17 = Abs(None=@n1)
  n18 = Multiply(A=@n17) {constB=0.001}
  n19 = Subtract(A=Ceil(None=@n17), B=@n17)
  n20 = Multiply(A=Abs(None=AppendVector(A=@n18, B=@n19))) {constB=2}
  n21 = Constant()
  n22 = Constant() {r=1}
  n23 = Constant() {r=-1}
  n24 = Multiply(
   A=DeriveNormalZ(InXY=Subtract(A=@n20))
   B=AppendVector(A=Constant2Vector() {g=1, r=1}, B=If(A=@n1, B=@n21, "A > B"=@n22, "A == B"=@n22, "A < B"=@n23))
  )
  FunctionOutput(None=MaterialFunctionCall(Vector=@n24) {materialFunction=&SafeNormalize}.Result) {outputName="( z sign ) XXX.YYY Normal", sortPriority=2}
  FunctionOutput(None=Normalize(VectorInput=@n16)) {outputName=XXY.YZZ Normalized Vector}
  MaterialFunctionCall!(Vector3=@n10.Result, MaximumNumberOfDigits, UVs, DebugTextLocation RG_UpperRight BA_LowerLeft, Component Spacing) {materialFunction=/Engine/Functions/Engine_MaterialFunctions02/Utility/DebugFloat3Values}
  MaterialFunctionCall(Vector3=@n24) {materialFunction=&DebugFloat3Values}
  MaterialFunctionCall(Vector3=Constant4Vector() {constant=<rgba 1 -0.9957 0.01 0>}.RGBA) {materialFunction=&DebugFloat3Values}
  n25 = Constant3Vector() {constant=<rgba -0.563 -0.6702 -0.4836 0>}
  MaterialFunctionCall!(
   Vector4=AppendVector(A=Subtract(A=@n16, B=@n25), B=@n1)
   MaximumNumberOfDigits
   UVs
   DebugTextLocation RG_UpperRight BA_LowerLeft
   Component Spacing
  ) {materialFunction=/Engine/Functions/Engine_MaterialFunctions02/Utility/DebugFloat4Values}
  Subtract(A=@n24, B=Constant3Vector() {constant=<rgba -0.5912 -0.5453 -0.5943 0>})
 ]
}
