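# Material function that plays VertexAnimationTools morph-target animation on a static mesh from baked position and normal textures.
# The animation value selects two adjacent texture rows via GradientMap_Multi, sampled at the morph UV and blended by its fraction.
# Blended offsets, transformed from instance space, drive World Position Offset.
# Vertex-shader normals pass through Custom UVs 2 and 3 to rebuild pixel-shader normals, with world space preferred over tangent space.
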MaterialFunction /Engine/Functions/Engine_MaterialFunctions02/MAXScripts/MS_VertexAnimationTools_MorphTargets
graph {
 properties {
  bExposeToLibrary=true
  description="This material function will apply the VertexAnimationTools morph target to an appropriate static mesh. "
  libraryCategoriesText [MAXScripts]
 }
 nodes [
  n1 = FunctionInput(Preview=TextureCoordinate() {coordinateIndex=1}) {
   bUsePreviewValueAsDefault=true
   description=This value should remain UV 1 unless the mesh was processed with another uv coordinate index. This feature was added after 4.13.
   inputName=Morph Target UV
   inputType=FunctionInput_Vector2
   sortPriority=1000
  }
  n2 = MaterialFunctionCall(Float2=@n1) {materialFunction=&BreakOutFloat2Components}
  n3 = FunctionInput() {
   bUsePreviewValueAsDefault=true
   description=1-based number indicating the number of morph targets in the morph texture.
   inputName=Number of Morph Targets
   inputType=FunctionInput_Scalar
   sortPriority=2
   previewValue=<64 0 0 1>
  }
  n4 = FunctionInput(Preview=MaterialFunctionCall(Speed=Divide(B=@n3) {constA=4, desc=2 sec}) {materialFunction=&TimeWithSpeedVariable}.Frac Time) {
   bUsePreviewValueAsDefault=true
   description="0-based number that is used to indicate what stage of the morph target animation is being used. If \"0-1 Animation Value?\" is set to true then 1 will represent the last frame of the animation otherwise the last frame will be equal to the number of frames -1."
   inputName=Morph Animation
   inputType=FunctionInput_Scalar
   sortPriority=-1
  }
  n5 = FunctionInput(Preview=StaticBool() {value=true}) {
   bUsePreviewValueAsDefault=true
   description=Defaults to true. Alternative is that each full number value will transition 1 morph target
   inputName=0-1 Animation Value?
   inputType=FunctionInput_StaticBool
   sortPriority=3
  }
  n6 = StaticSwitch(True=Multiply(A=@n4, B=Subtract(A=@n3)), False=@n4, Value=@n5)
  n7 = Floor(None=@n6)
  n8 = TextureObject() {samplerType=SAMPLERTYPE_LinearColor, texture=/Engine/Functions/Engine_MaterialFunctions02/ExampleContent/TextureBasedWPO/DemoBox_Normals}
  n9 = FunctionInput(Preview=@n8) {
   bUsePreviewValueAsDefault=true
   description=Input the morph texture ending with _Normal.
   inputName=Morph Normal
   inputType=FunctionInput_Texture2D
  }
  n10 = Add(A=@n7)
  n11 = Frac(None=@n6)
  n12 = LinearInterpolate(
   A=MaterialFunctionCall!(Greyscale Value To Gradient Map=@n2.R, "Index (0 Based)"=@n7, Gradient Texture=@n9, Number Of Gradients In the Gradient Map=@n3) {materialFunction=/Engine/Functions/Engine_MaterialFunctions02/Gradients/GradientMap_Multi}.RGB
   B=MaterialFunctionCall(Greyscale Value To Gradient Map=@n2.R, "Index (0 Based)"=@n10, Gradient Texture=@n9, Number Of Gradients In the Gradient Map=@n3) {materialFunction=&GradientMap_Multi}.RGB
   Alpha=@n11
  )
  n13 = Normalize(VectorInput=Transform(None=ConstantBiasScale(None=@n12) {bias=-0.5, scale=2}) {transformSourceType=TRANSFORMSOURCE_Instance})
  n14 = MaterialFunctionCall(Float3=@n13) {materialFunction=&BreakOutFloat3Components}
  FunctionOutput(None=AppendVector(A=@n14.B, B=Constant())) {outputName="Custom UV 3 (Needed)", sortPriority=3}
  n15 = AppendVector(
   A=TextureCoordinate() {coordinateIndex=2}
   B=MaterialFunctionCall(Float2=TextureCoordinate() {coordinateIndex=3}) {materialFunction=&BreakOutFloat2Components}.R
  )
  FunctionOutput(None=Normalize(VectorInput=Transform(None=@n15) {transformSourceType=TRANSFORMSOURCE_World, transformType=TRANSFORM_Tangent})) {
   description="Note: Tangent Space normals may create surface artifacts. Ideally, one should use the world space  normal output above and plug it directly into the normals material input. Tangent Space Normals should be unchecked in the shader when using world space normals. \n\nThis value only puts out the meshes deformed vertex normal if the custom uv inputs are attached. \n\nAlso, one mustn't use vertex normal in their base shader when applying this technique. The vertex normals do not get updated and therefore that node will produce incorrect results.\n\n"
   outputName="Pixel Shader Tangent Vertex Normal (See tooltip)"
  }
  FunctionOutput(None=AppendVector(A=@n14.R, B=@n14.R)) {
   description="Connecting this Output pin to Custom UV 2 will enable the normal output pin to work correctly.\n\nThis will allow you to sample the noraml texture in the vertex shader and then pass it to the normal input in the pixel shader. "
   outputName="Custom UV 2 (Needed)"
   sortPriority=2
  }
  FunctionOutput(None=Normalize(VectorInput=@n15)) {
   description="Note: Due to certain limitations, this output pin should be used instead of tangent space normals. Transforming the normals to tangent space should be avoided. Plug this output pin directly into the materials normal input and UNCHECK Tangent Space  Normals in the materials properties. \n\nThis value only puts out the meshes deformed vertex normal if the custom uv inputs are attached.\n\nAlso, one mustn't use vertex normal in their base shader when applying this technique. The vertex normals do not get updated and therefore that node will produce incorrect results.\n"
   outputName="Pixel Shader World Vertex Normal (See tooltip)"
   sortPriority=-1
  }
  n16 = FunctionInput(Preview=TextureObject() {samplerType=SAMPLERTYPE_LinearColor, texture=/Engine/Functions/Engine_MaterialFunctions02/ExampleContent/TextureBasedWPO/DemoBox}) {bUsePreviewValueAsDefault=true, description=Insert the .EXR texture here., inputName=Morph Texture, inputType=FunctionInput_Texture2D}
  n17 = MaterialFunctionCall(Greyscale Value To Gradient Map=@n2.R, "Index (0 Based)"=@n7, Gradient Texture=@n16, Number Of Gradients In the Gradient Map=@n3) {materialFunction=&GradientMap_Multi}
  n18 = MaterialFunctionCall(Greyscale Value To Gradient Map=@n2.R, "Index (0 Based)"=@n10, Gradient Texture=@n16, Number Of Gradients In the Gradient Map=@n3) {materialFunction=&GradientMap_Multi}
  n19 = LinearInterpolate(A=AppendVector(A=@n17.RGB, B=@n17.RGB), B=AppendVector(A=@n18.RGB, B=@n18.RGB), Alpha=@n11)
  FunctionOutput(None=Transform(None=ComponentMask(None=@n19) {b=true, g=true, r=true}) {transformSourceType=TRANSFORMSOURCE_Instance}) {description=Outputs morph target offsets, outputName=World Position Offset, sortPriority=1}
  FunctionOutput(None=@n13) {
   description="This feature won't be needed for most uses. If you do however need the deformed normal in the vertex shader, then you should use this value instead of the other normal output. This cannot be used in the pixel shader. "
   outputName="Vert shader Normal (Advanced)"
   sortPriority=5
  }
  MaterialFunctionCall!(Scalar Value=ComponentMask(None=@n19) {a=true}) {materialFunction=/Engine/Functions/Engine_MaterialFunctions02/Normals/UnpackNormalFromFloat}
 ]
}
